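# Physics asset for the F_MED_Shiny medium female player body cosmetic, covering only the dyn_chain_lock bone and the upper spine.
# Each of the two bones carries a single capsule: a slim, tilted one on dyn_chain_lock and a much wider one on spine_05.
# One constraint links dyn_chain_lock to spine_05 with swing and twist limit angles, apparently so the lock accessory can dangle against the torso.
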PhysicsAsset /BRCosmetics/Characters/Player/Female/Medium/Bodies/F_MED_Shiny/Meshes/F_MED_Shiny_Body_Physics
bodies {
 dyn_chain_lock [{shapeType=Capsule, length=14.1811, radius=3.3003, shapeName=dyn_chain_lock_capsule, center=<-3.329 0 3.8718>, rotation=<rot 73.9364 0 0>}]
 spine_05 [{shapeType=Capsule, length=13.6099, radius=16.3901, shapeName=spine_05_capsule, center=<-7.6152 -3.824 0>}]
}
constraints [{bone1Name=dyn_chain_lock, bone2Name=spine_05, swing1LimitDegrees=45, swing2LimitDegrees=45, twistLimitDegrees=33.3333}]
