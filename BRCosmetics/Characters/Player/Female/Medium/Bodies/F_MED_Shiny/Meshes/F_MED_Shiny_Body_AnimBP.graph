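# Animation Blueprint for the F_MED_Shiny female body cosmetic that follows the character's pose and adds secondary physics motion.
# It copies the pose from the attached parent mesh, then runs a RigidBody simulation using the body's physics asset in pelvis bone space with overridden gravity.
# Two Anim Dynamics nodes for chain lock and top bones sit unconnected to the output.
# The event graph holds only an empty update event.
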AnimBlueprint /BRCosmetics/Characters/Player/Female/Medium/Bodies/F_MED_Shiny/Meshes/F_MED_Shiny_Body_AnimBP
target_skeleton=&/F_MED_Shiny_Body_Skeleton
graphs [
 AnimGraph {nodes [
   Anim Dynamics - Bone: dyn_chain_lock(Alpha=1, ExternalForce="0,0,0") {
    angularDampingOverride=0.96
    bGravityOverrideInSimSpace=true
    bLinearSpring=true
    bOverrideAngularDamping=true
    bOverrideLinearDamping=true
    bUseSphericalLimits=true
    linearDampingOverride=0.96
    linearSpringConstant=700
    alphaBoolBlend {blendOption=Linear}
    boundBone {boneName=dyn_chain_lock}
    componentLinearAccScale=<1 1 1>
    componentLinearVelScale=<1 1 1>
    physicsBodyDefinitions [{
      sphereCollisionRadius=0
      boundBone {boneName=dyn_chain_lock}
      boxExtents=<8 5 6>
      constraintSetup {angularTargetAxis=AxisY, twistAxis=AxisY, angularLimitsMax=<20 10 30>, angularLimitsMin=<-10 -10 -10>, angularTarget=<0 0 0>}
      localJointOffset=<3 -2 0>
     }]
   }
   Anim Dynamics - Bone: dyn_chain_Top(Alpha=1, ExternalForce="0,0,0") {
    angularBiasOverride=0.96
    bGravityOverrideInSimSpace=true
    bOverrideAngularDamping=true
    bOverrideLinearDamping=true
    linearDampingOverride=0.96
    simulationSpace=Actor
    alphaBoolBlend {blendOption=Linear}
    boundBone {boneName=dyn_chain_Top}
    componentLinearAccScale=<1 1 1>
    componentLinearVelScale=<1 1 1>
    physicsBodyDefinitions [{
      sphereCollisionRadius=0
      boundBone {boneName=dyn_chain_Top}
      boxExtents=<1 1 1>
      constraintSetup {angularLimitsMax=<0 0 0>, angularLimitsMin=<0 0 0>, angularTarget=<0 0 0>}
      localJointOffset=<0 0 0>
     }]
   }
   n1 = RigidBody(ComponentPose=Local To Component(LocalPose=Copy Pose From Mesh() {bUseAttachedParent=true})) {
    bOverrideWorldGravity=true
    overridePhysicsAsset=&/F_MED_Shiny_Body_Physics
    simulationSpace=BaseBoneSpace
    alphaBoolBlend {blendOption=Linear}
    baseBoneRef {boneName=pelvis}
    componentAppliedLinearAccClamp=<1000 1000 1000>
    componentLinearAccScale=<5 5 5>
    componentLinearVelScale=<5 5 5>
    overrideWorldGravity {z=-980}
   }
   Output Pose(Result=Component To Local(ComponentPose=@n1)) {layerGroup=None}
  ]}
 EventGraph {nodes [Event BlueprintUpdateAnimation()]}
]
variables []
