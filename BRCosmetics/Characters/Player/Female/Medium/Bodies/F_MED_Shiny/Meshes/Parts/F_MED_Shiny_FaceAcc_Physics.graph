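# Physics asset for the face accessory mesh of the F_MED_Shiny player cosmetic, setting up physics for bangs and earrings.
# Left and right bangs form chains of capsule bodies hanging from the head, with swing and twist limit angles on each joint.
# Head, neck, clavicle and upper-spine shapes provide body volumes, linked by free angular constraints.
# The sphere earrings attach to the head asymmetrically: the right is angularly locked, while the left rotates freely.
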PhysicsAsset /BRCosmetics/Characters/Player/Female/Medium/Bodies/F_MED_Shiny/Meshes/Parts/F_MED_Shiny_FaceAcc_Physics
bodies {
 clavicle_l [{shapeType=Capsule, length=7.5801, radius=6.9382, center=<5.2163 -1.2987 -2.8186>, rotation=<rot 90 0 0>}]
 clavicle_r [{shapeType=Capsule, length=7.5801, radius=6.9382, center=<-5.2163 1.2987 2.8186>, rotation=<rot 90 0 0>}]
 dyn_bangs_left_01 [{shapeType=Capsule, length=4.25, radius=1.5, center=<0.0767 0 -0.4847>, rotation=<rot 90 -0.0198 -0.0198>}]
 dyn_bangs_left_02 [{shapeType=Capsule, length=2.25, radius=1.5, center=<0 -0.0001 -0.0842>, rotation=<rot 90 -0.0396 -0.0396>}]
 dyn_bangs_left_03 [{shapeType=Capsule, length=2.75, radius=2, center=<1.7419 0.0001 0.4385>, rotation=<rot 90 0 0>}]
 dyn_bangs_left_top [{shapeType=Capsule, length=0.5, radius=3}]
 dyn_bangs_right_01 [{shapeType=Capsule, length=1.25, radius=2, center=<-0.6897 -0.5481 0.0252>, rotation=<rot 90 0 -10.0021>}]
 dyn_bangs_right_02 [{shapeType=Capsule, length=2.25, radius=2, center=<-0.8257 0.0143 0.2866>, rotation=<rot 90 -0.0198 -0.0198>}]
 dyn_bangs_right_03 [{shapeType=Capsule, length=3, radius=1.75, center=<0.3415 0.0045 0.4616>, rotation=<rot 86.5443 -99.8275 -80.1246>}]
 dyn_bangs_right_top [{shapeType=Capsule, length=0, radius=3.75, center=<-1.4347 0.0134 -1.1492>, rotation=<rot 0 10.0004 0>}]
 dyn_bnags_right_04 [{shapeType=Capsule, length=3, radius=1, center=<1.7184 -1.0227 -0.5263>, rotation=<rot -90 -172.875 152.8727>}]
 dyn_earing_left [{shapeType=Sphere, radius=2.4994, center=<-2.3358 -0.4123 0.1086>}]
 dyn_earing_right [{shapeType=Sphere, radius=3, center=<-2.2533 0.0697 -0.4036>}]
 head [{shapeType=Capsule, length=1, radius=1}]
 neck_02 [{shapeType=Capsule, length=10.9843, radius=4.6791, rotation=<rot 0 -89.9999 -89.9996>}]
 spine_05 [
  {shapeType=Capsule, length=14, radius=6, center=<0.1411 1.1592 -5.8362>, rotation=<rot -84.9994 0.0162 -50.0162>}
  {shapeType=Capsule, length=14, radius=6, center=<-0.5445 1.3555 4.7476>, rotation=<rot 82.031 62.5276 112.8203>}
  {shapeType=Box, extentX=14.0671, extentY=7.4811, extentZ=18.7407, center=<2.9977 3.4794 0.0004>, rotation=<rot 0 -39.9975 0>}
 ]
}
constraints [
 {bone1Name=dyn_bangs_right_top, bone2Name=head, swing1LimitDegrees=50, swing2LimitDegrees=50, twistLimitDegrees=10}
 {bone1Name=dyn_bangs_right_01, bone2Name=dyn_bangs_right_top, swing1LimitDegrees=40, swing2LimitDegrees=40, twistLimitDegrees=10}
 {bone1Name=dyn_bangs_right_02, bone2Name=dyn_bangs_right_01, swing1LimitDegrees=50, swing2LimitDegrees=50, twistLimitDegrees=10}
 {bone1Name=dyn_bangs_right_03, bone2Name=dyn_bangs_right_02, swing1LimitDegrees=50, swing2LimitDegrees=50, twistLimitDegrees=10}
 {bone1Name=dyn_bnags_right_04, bone2Name=dyn_bangs_right_03, swing1LimitDegrees=50, swing2LimitDegrees=50, twistLimitDegrees=10}
 {bone1Name=dyn_bangs_left_top, bone2Name=head, swing1LimitDegrees=50, swing2LimitDegrees=50, twistLimitDegrees=10}
 {bone1Name=dyn_bangs_left_01, bone2Name=dyn_bangs_left_top, swing1LimitDegrees=50, swing2LimitDegrees=50, twistLimitDegrees=10}
 {bone1Name=dyn_bangs_left_02, bone2Name=dyn_bangs_left_01, swing1LimitDegrees=50, swing2LimitDegrees=50, twistLimitDegrees=10}
 {bone1Name=dyn_bangs_left_03, bone2Name=dyn_bangs_left_02, swing1LimitDegrees=50, swing2LimitDegrees=50, twistLimitDegrees=10}
 {
  bone1Name=head
  bone2Name=neck_02
  swing1LimitDegrees=45
  swing1Motion=Free
  swing2LimitDegrees=45
  swing2Motion=Free
  twistLimitDegrees=45
  twistMotion=Free
 }
 {
  bone1Name=clavicle_r
  bone2Name=spine_05
  swing1LimitDegrees=45
  swing1Motion=Free
  swing2LimitDegrees=45
  swing2Motion=Free
  twistLimitDegrees=45
  twistMotion=Free
 }
 {
  bone1Name=clavicle_l
  bone2Name=spine_05
  swing1LimitDegrees=45
  swing1Motion=Free
  swing2LimitDegrees=45
  swing2Motion=Free
  twistLimitDegrees=45
  twistMotion=Free
 }
 {
  bone1Name=neck_02
  bone2Name=spine_05
  swing1LimitDegrees=45
  swing1Motion=Free
  swing2LimitDegrees=45
  swing2Motion=Free
  twistLimitDegrees=45
  twistMotion=Free
 }
 {
  bone1Name=dyn_earing_right
  bone2Name=head
  swing1LimitDegrees=45
  swing1Motion=Locked
  swing2LimitDegrees=45
  swing2Motion=Locked
  twistLimitDegrees=45
  twistMotion=Locked
 }
 {
  bone1Name=dyn_earing_left
  bone2Name=head
  swing1LimitDegrees=45
  swing1Motion=Free
  swing2LimitDegrees=45
  swing2Motion=Free
  twistLimitDegrees=45
  twistMotion=Free
 }
]
