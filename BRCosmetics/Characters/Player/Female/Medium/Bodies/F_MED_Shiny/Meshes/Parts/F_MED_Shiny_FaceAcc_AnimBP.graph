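# Post-pose physics AnimBP for the Shiny medium female's face accessory mesh, adding secondary motion to earrings and hair.
# It copies the attached parent's pose, then chains spring-based AnimDynamics on earrings, hair strands and hair bases, with a spine_05-space RigidBody node mid-chain.
# Pawn 2D speed scales hair-base alpha from partial at rest to full when moving.
# Two alternative left hair-base nodes are unconnected, and the EventGraph does nothing.
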AnimBlueprint /BRCosmetics/Characters/Player/Female/Medium/Bodies/F_MED_Shiny/Meshes/Parts/F_MED_Shiny_FaceAcc_AnimBP
target_skeleton=&/F_MED_Shiny_FaceAcc_Skeleton
graphs [
 AnimGraph {nodes [
   Anim Dynamics - Bone: dyn_hair_base_left(Alpha=1, ExternalForce="0,0,0") {
    bGravityOverrideInSimSpace=true
    bLinearSpring=true
    linearDampingOverride=0.98
    linearSpringConstant=500
    alphaBoolBlend {blendOption=Linear}
    boundBone {boneName=dyn_hair_base_left}
    componentLinearAccScale=<2 2 2>
    componentLinearVelScale=<2 2 2>
    physicsBodyDefinitions [{
      sphereCollisionRadius=0
      boundBone {boneName=dyn_hair_base_left}
      boxExtents=<10 10 10>
      constraintSetup {angularTargetAxis=AxisY, twistAxis=AxisY, angularLimitsMax=<20 20 0>, angularLimitsMin=<-20 -20 0>, angularTarget=<0 0 0>}
      localJointOffset=<-2 0 2>
     }]
   }
   Anim Dynamics - Bone: dyn_hair_base_left(Alpha=1, ExternalForce="0,0,0") {
    angularSpringConstant=1000
    bAngularSpring=true
    bGravityOverrideInSimSpace=true
    linearDampingOverride=0.9
    simulationSpace=Actor
    alphaBoolBlend {blendOption=Linear}
    boundBone {boneName=dyn_hair_base_left}
    componentLinearAccScale=<1 1 1>
    componentLinearVelScale=<0.15 0.15 0.15>
    physicsBodyDefinitions [{
      sphereCollisionRadius=0
      boundBone {boneName=dyn_hair_base_left}
      boxExtents=<30 30 10>
      constraintSetup {angularTargetAxis=AxisY, twistAxis=AxisY, angularLimitsMax=<20 20 0>, angularLimitsMin=<-20 -20 0>, angularTarget=<0 0 0>}
      localJointOffset=<-2 0 2>
     }]
   }
   n1 = Anim Dynamics - Bone: dyn_earing_right(ComponentPose=Local To Component(LocalPose=Copy Pose From Mesh() {bUseAttachedParent=true}), Alpha=1, ExternalForce="0,0,0") {
    angularDampingOverride=0.9
    angularSpringConstant=1000
    bAngularSpring=true
    bGravityOverrideInSimSpace=true
    alphaBoolBlend {blendOption=Linear}
    boundBone {boneName=dyn_earing_right}
    componentLinearAccScale=<1.3 1.3 1.3>
    componentLinearVelScale=<1 1 1>
    physicsBodyDefinitions [{
      sphereCollisionRadius=0
      boundBone {boneName=dyn_earing_right}
      boxExtents=<20 3 20>
      constraintSetup {twistAxis=AxisY, angularLimitsMax=<0 0 80>, angularLimitsMin=<0 0 -10>, angularTarget=<0 0 0>}
      localJointOffset=<-1 0 0>
     }]
    relativeSpaceBone {boneName=pelvis}
   }
   n2 = Anim Dynamics - Bone: dyn_earing_left(ComponentPose=@n1, Alpha=1, ExternalForce="0,0,0") {
    angularDampingOverride=0.9
    angularSpringConstant=1000
    bAngularSpring=true
    bGravityOverrideInSimSpace=true
    alphaBoolBlend {blendOption=Linear}
    boundBone {boneName=dyn_earing_left}
    componentLinearAccScale=<1.3 1.3 1.3>
    componentLinearVelScale=<1 1 1>
    physicsBodyDefinitions [{
      sphereCollisionRadius=0
      boundBone {boneName=dyn_earing_left}
      boxExtents=<20 3 20>
      constraintSetup {twistAxis=AxisY, angularLimitsMax=<0 0 80>, angularLimitsMin=<0 0 -10>, angularTarget=<0 0 0>}
      localJointOffset=<-1 0 0>
     }]
    relativeSpaceBone {boneName=pelvis}
   }
   n3 = Anim Dynamics - Bone: dyn_hair_strand_right(ComponentPose=@n2, Alpha=1, ExternalForce="0,0,0") {
    angularSpringConstant=900
    bAngularSpring=true
    bGravityOverrideInSimSpace=true
    linearDampingOverride=0.9999
    simulationSpace=Actor
    alphaBoolBlend {blendOption=Linear}
    boundBone {boneName=dyn_hair_strand_right}
    componentLinearAccScale=<1 1 1>
    componentLinearVelScale=<0.15 0.15 0.15>
    physicsBodyDefinitions [{
      sphereCollisionRadius=0
      boundBone {boneName=dyn_hair_strand_right}
      boxExtents=<30 30 10>
      constraintSetup {angularTargetAxis=AxisY, twistAxis=AxisY, angularLimitsMax=<5 20 0>, angularLimitsMin=<-5 -20 0>, angularTarget=<0 0 0>}
      localJointOffset=<-2 0 2>
     }]
   }
   n4 = MapRangeClamped(Value=Get PawnSpeed2D(), InRangeA=0, InRangeB=100, OutRangeA=.2)
   n5 = Anim Dynamics - Bone: dyn_hair_base_right(ComponentPose=@n3, Alpha=@n4, ExternalForce="0,0,0") {
    angularDampingOverride=0.96
    angularSpringConstant=500
    bAngularSpring=true
    bGravityOverrideInSimSpace=true
    bLinearSpring=true
    bOverrideAngularDamping=true
    gravityScale=0.5
    linearSpringConstant=500
    simulationSpace=BoneRelative
    alphaBoolBlend {blendOption=Linear}
    boundBone {boneName=dyn_hair_base_right}
    componentAppliedLinearAccClamp=<600 600 600>
    componentLinearAccScale=<1 1 1>
    componentLinearVelScale=<1 1 1>
    physicsBodyDefinitions [{
      sphereCollisionRadius=0
      boundBone {boneName=dyn_hair_base_right}
      boxExtents=<4 4 4>
      constraintSetup {angularTargetAxis=AxisY, angularLimitsMax=<0 15 15>, angularLimitsMin=<0 -15 -15>, angularTarget=<0 0 0>}
      localJointOffset=<0 0 -1>
     }]
    relativeSpaceBone {boneName=root}
   }
   n6 = RigidBody(ComponentPose=@n5) {
    bOverrideWorldGravity=true
    overridePhysicsAsset=&/F_MED_Shiny_FaceAcc_Physics
    simulationSpace=BaseBoneSpace
    alphaBoolBlend {blendOption=Linear}
    baseBoneRef {boneName=spine_05}
    componentLinearAccScale=<0.2 0.2 0.2>
    componentLinearVelScale=<0.2 0.2 0.2>
    overrideWorldGravity {z=-980}
   }
   n7 = Anim Dynamics - Bone: dyn_hair_strand_left(ComponentPose=@n6, Alpha=1, ExternalForce="0,0,0") {
    angularSpringConstant=800
    bAngularSpring=true
    bGravityOverrideInSimSpace=true
    linearDampingOverride=0.9999
    simulationSpace=Actor
    alphaBoolBlend {blendOption=Linear}
    boundBone {boneName=dyn_hair_strand_left}
    componentLinearAccScale=<1 1 1>
    componentLinearVelScale=<0.15 0.15 0.15>
    physicsBodyDefinitions [{
      sphereCollisionRadius=0
      boundBone {boneName=dyn_hair_strand_left}
      boxExtents=<20 20 5>
      constraintSetup {angularTargetAxis=AxisY, twistAxis=AxisY, angularLimitsMax=<10 10 0>, angularLimitsMin=<-10 -10 0>, angularTarget=<0 0 0>}
      localJointOffset=<-2 0 2>
     }]
   }
   n8 = Anim Dynamics - Bone: dyn_hair_base_left(ComponentPose=@n7, Alpha=@n4, ExternalForce="0,0,0") {
    angularDampingOverride=0.96
    angularSpringConstant=500
    bAngularSpring=true
    bGravityOverrideInSimSpace=true
    bLinearSpring=true
    bOverrideAngularDamping=true
    gravityScale=0.5
    linearSpringConstant=500
    simulationSpace=BoneRelative
    alphaBoolBlend {blendOption=Linear}
    boundBone {boneName=dyn_hair_base_left}
    componentAppliedLinearAccClamp=<600 600 600>
    componentLinearAccScale=<1 1 1>
    componentLinearVelScale=<1 1 1>
    physicsBodyDefinitions [{
      sphereCollisionRadius=0
      boundBone {boneName=dyn_hair_base_left}
      boxExtents=<4 4 4>
      constraintSetup {angularTargetAxis=AxisY, angularLimitsMax=<0 15 15>, angularLimitsMin=<0 -15 -15>, angularTarget=<0 0 0>}
      localJointOffset=<0 0 -1>
     }]
    relativeSpaceBone {boneName=root}
   }
   Output Pose(Result=Component To Local(ComponentPose=@n8)) {layerGroup=None}
  ]}
 EventGraph {nodes [
   TryGetPawnOwner()
   Event BlueprintUpdateAnimation()
  ]}
]
variables []
